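; Black Monday Jam interactable prop, apparently for a quest, that supplies its own interaction time to interacting pawns.
; Its construction script creates a dynamic emissive blink-light material on the jam mesh and swaps in a wall mesh when flagged as wall-attached.
; A helper function activates the jam's vision component only when the given player pawn owns a matching cosmetic gameplay tag; no caller for it appears here.
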(fn BlueprintGetInteractionTime
  (params (InteractingPawn :type Fort%20Pawn%20Object%20Reference) (InteractionBeingAttempted :type EInteractionBeingAttempted%20Enum :default "FirstInteraction"))
  (return (Variables|Default|GetInteractTime) true))

(fn NewFunction_0
  (params (PlayerPawn :type Fort%20Player%20Pawn%20Athena%20Object%20Reference))
  (if (GameplayTags|HasAnyTags (GameplayTags|GetOwnedGameplayTags PlayerPawn :!class /Script/GameplayTags.GameplayTagAssetInterface) (Variables|Default|GetCosmeticCustomTag))
    (Components|Activation|Activate (Variables|Default|GetP_BlackMonday_Jam_Vision) false)
    (else
      _)))

(fn ConstructionScript
  (|CallParentFunction)
  (Utilities|FlowControl|Sequence
    (:then_0
      ;[1 1 1] EmissiveLight
      (bind _returnvalue (Rendering|Material|CreateDynamicMaterialInstance (Variables|Default|GetSM_Jam) 1 "/Figment_S10/Gadgets/Assets/BlackMonday_Jam/Materials/MI_BlackMonday_Jam_Propane_Tank.MI_BlackMonday_Jam_Propane_Tank" "None" :!class /Script/Engine.PrimitiveComponent))
      (Variables|Art|Mat|SetLightBlinkMat _returnvalue))
    (:then_1
      (if (Variables|Default|GetbIsWallAttachedMesh)
        (Components|StaticMesh|SetStaticMesh (Variables|Default|GetSM_Jam) (Variables|Art|GetSM_Jam_Wall))))))
