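; Interactable quest lamp for the Black Monday Strawberry challenge that switches on for a timed period, then becomes usable again after switching off.
; Interaction stores the pawn and clears LampOff, whose RepNotify runs client-side timelines that tilt, turn and flicker the lamp, with on/off sounds.
; A local player's first use also plays a stinger.
; A downward trace at BeginPlay binds the supporting building actor's death to destroy the lamp.
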(event EventBeginPlay
  ;[1 1 1] BEGIN PLAY: Make sure lamp is off
  (Variables|Default|SetLampOff true)
  (Utilities|FlowControl|Delay 1.0)
  ;[1 1 1] Destroy Lamp if whatever it's sitting on is destroyed
  (bind (_outhit _returnvalue) (Collision|LineTraceForObjects :Start (+ (Transformation|GetActorLocation) "0, 0,200.000000") :End (+ (Transformation|GetActorLocation) "0, 0,-200.000000") :ObjectTypes (Utilities|Array|MakeArray "ObjectTypeQuery1") :bTraceComplex false :DrawDebugType "None" :bIgnoreSelf true :TraceColor "(R=1.000000,G=0.000000,B=0.000000,A=1.000000)" :TraceHitColor "(R=0.000000,G=1.000000,B=0.000000,A=1.000000)" :DrawTime 5e+01))
  (if _returnvalue
    (bind _asbuilding_actor (Utilities|Casting|CastToBuildingActor (.HitActor (Collision|BreakHitResult _outhit)))
      (:then
        (Development|PrintString (Utilities|GetDisplayName _asbuilding_actor) false false "(R=1.000000,G=0.597294,B=0.000000,A=1.000000)" 2e+01)
        (EventDispatchers|BindEventtoOnDied _asbuilding_actor (delegate Custom|Lamp%20Base%20Destroyed)))
      (:CastFailed
        (Development|PrintString "Cast Failed" false false "(R=1.000000,G=0.079394,B=0.000000,A=1.000000)" 2e+01)))
    (else
      (Development|PrintString "Nothing Hit" false false "(R=1.000000,G=0.079394,B=0.000000,A=1.000000)" 2e+01))))

(event Custom|Lamp%20Base%20Destroyed
  (params (DamagedActor :type Actor%20Object%20Reference) (Damage :type Float%20%28single-precision%29) (InstigatedBy :type Controller%20Object%20Reference) (DamageCauser :type Actor%20Object%20Reference) (HitLocation :type Vector) (FHitComponent :type Primitive%20Component%20Object%20Reference) (BoneName :type Name) (Momentum :type Vector))
  ;[1 1 1] Destroy Lamp if whatever it's sitting on is destroyed
  (Development|PrintString "DESTRUCTION!!!!" false false)
  (Actor|DestroyActor))

(event Interaction|EventBlueprintOnInteract
  (params (InteractingPawn :type Fort%20Pawn%20Object%20Reference) (InteractionBeingAttempted :type EInteractionBeingAttempted%20Enum :default "FirstInteraction"))
  ;[1 1 1] INTERACTION: set timer and hit rep notify's to pass info to client
  (Variables|Default|SetInteractingPawn InteractingPawn)
  (Variables|Default|SetInteractingPawnLoc (Transformation|GetActorLocation InteractingPawn))
  (Variables|Default|SetLampOff false)
  (bind _returnvalue (Utilities|Time|SetTimerbyEvent (delegate Custom|LampTimerUp) (Variables|Default|GetLampOnTime)))
  (Variables|Default|SetLampLightTimer _returnvalue))

(event Custom|LampTimerUp
  ;[1 1 1] Lamp timer up: Set Lamp notify to OFF
  (Variables|Default|SetLampOff true))

(event Custom|LampOnFlicker
  (bind _interactingpawn (Variables|Default|GetInteractingPawn))
  (bind _blackmonday_strawberry_fx_meshes (Variables|Default|GetBlackMonday_Strawberry_FX_Meshes))
  ;[1 1 1] Lamp Turns ON Event
  (Networking|SwitchHasAuthority
    (:Remote
      (Utilities|FlowControl|Sequence
        (:then_0
          ;[1 1 1] CLIENT ONLY: Rotate light up and in the direction the player is looking
          (bind (_direction _activate_time) (Timeline|Timeline_1 Play)
            (:Update
              (Transformation|SetRelativeRotation (Variables|Default|GetBlackMonday_Strawberry_Lamp) (Math|Rotator|Lerp(Rotator) "0, 0,-30.000000" "0, 0,35.000000" _activate_time) false false)
              (Transformation|SetWorldRotation (Variables|Default|GetBlackMonday_Strawberry_Arms) (Math|Rotator|Lerp(Rotator) "0, 0, 0" (Math|Rotator|MakeRotator 0.0 0.0 (.yaw (Math|Rotator|MakeRotfromZ (- (Transformation|GetActorLocation) (Variables|Default|GetInteractingPawnLoc))))) _activate_time) false false))
            (:Finished
              ;[1 1 1] Play Stinger if first time used
              (for _array_element (Variables|Default|GetAll_Interacting_Pawns)
                (if (Utilities|Equal(Object) _array_element (Variables|Default|GetInteractingPawn))
                  (Variables|Default|SetHasPawnUsedThisYet? true)))
              (if (Variables|Default|GetHasPawnUsedThisYet?)
                (Development|PrintString "Already Used")
                (else
                  (Utilities|Array|AddUnique (Variables|Default|GetAll_Interacting_Pawns) _interactingpawn)
                  (if (Utilities|Equal(Object) _interactingpawn (Game|Player|GetPlayerPawn (Fortnite|GetFirstLocalFortPlayerController true)))
                    (Audio|PlaySound2D "/Figment_S10/Athena/Sounds/Quests/BlackMonday/BlackMonday_Stinger_On_Cue.BlackMonday_Stinger_On_Cue")))))))
        (:then_1
          ;[1 1 1] Turn On Audio
          (Audio|PlaySoundatLocation (Variables|Default|GetTurnOnSound) (Transformation|GetActorLocation)))
        (:then_2
          (Utilities|FlowControl|Delay 0.5)
          ;[1 1 1] CLIENT ONLY: Flicker Light on and Off
          (bind (_direction_2 _newtrack_1) (Timeline|Timeline_0 Play)
            (:Update
              (Rendering|Material|SetVectorParameterValue (Variables|Default|GetMID_LampGlass) "EmissiveMultiplier" (Math|Conversions|ToLinearColor(Vector) (Math|Vector|Lerp(Vector) "0, 0, 0" "14.000000,8.000000,4.000000" _newtrack_1)) :!class /Script/Engine.MaterialInstanceDynamic))
            (:Finished
              (switch Utilities|FlowControl|Switch|SwitchonETimelineDirection _direction_2
                (:Backward
                  _)))
            (:LightOn
              (Development|SetHiddeninGame _blackmonday_strawberry_fx_meshes false false))
            (:LightOff
              _))
          (Development|SetHiddeninGame _blackmonday_strawberry_fx_meshes true false))))))

(event Custom|LampOffFlicker
  (bind _interactingpawn (Variables|Default|GetInteractingPawn))
  (bind _blackmonday_strawberry_fx_meshes (Variables|Default|GetBlackMonday_Strawberry_FX_Meshes))
  ;[1 1 1] Lamp Turns OFF Event
  (Utilities|FlowControl|Sequence
    (:then_0
      (Networking|SwitchHasAuthority
        (:Remote
          (Utilities|FlowControl|Sequence
            (:then_0
              ;[1 1 1] CLIENT ONLY: Rotate light up and in the direction the player is looking
              (Timeline|Timeline_1 Reverse))
            (:then_1
              ;[1 1 1] CLIENT ONLY: Flicker Light on and Off
              (Timeline|Timeline_0 Reverse))
            (:then_2
              ;[1 1 1] Turn Off Audio
              (Audio|PlaySoundatLocation (Variables|Default|GetTurnOffSound) (Transformation|GetActorLocation)))))))
    (:then_1
      (Utilities|FlowControl|Delay 1.0)
      (Variables|Default|SetInteractable true))))

(fn BlueprintCanInteract
  (params (InteractingPawn :type Fort%20Pawn%20Object%20Reference) (InteractionBeingAttempted :type EInteractionBeingAttempted%20Enum :default "FirstInteraction") (InteractionType :type TInteractionType%20Enum :default "IT_NoInteraction"))
  (return (Variables|Default|GetInteractable)))

(fn BlueprintGetInteractionString
  (params (InteractingPawn :type Fort%20Pawn%20Object%20Reference) (InteractionBeingAttempted :type EInteractionBeingAttempted%20Enum :default "FirstInteraction"))
  (return (Variables|Default|GetLightLampString)))

(fn OnRep_InteractingPawn)

(fn OnRep_LampOff
  (if (Variables|Default|GetLampOff)
    (CallFunction|LampOffFlicker)
    (Development|PrintString "Interactable ON" true true "(R=0.249851,G=1.000000,B=0.000000,A=1.000000)")
    (else
      (CallFunction|LamponFlicker)
      (Variables|Default|SetInteractable false)
      (Development|PrintString "Interactable OFF" true true "(R=1.000000,G=0.069961,B=0.000000,A=1.000000)"))))

(fn ConstructionScript
  (bind _blackmonday_strawberry_lamp (Variables|Default|GetBlackMonday_Strawberry_Lamp))
  (Utilities|FlowControl|Sequence
    (:then_0
      (bind _returnvalue (Rendering|Material|CreateDynamicMaterialInstance _blackmonday_strawberry_lamp 0 "/Game/Gadgets/Assets/BlackMonday_Strawberry/Materials/MI_BlackMonday_Strawberry.MI_BlackMonday_Strawberry" "None" :!class /Script/Engine.PrimitiveComponent))
      (Variables|Default|SetMID_LampLight _returnvalue))
    (:then_1
      (bind _returnvalue_2 (Rendering|Material|CreateDynamicMaterialInstance _blackmonday_strawberry_lamp 1 "/Game/Gadgets/Assets/BlackMonday_Strawberry/Materials/MI_BlackMonday_Strawberry_Glass.MI_BlackMonday_Strawberry_Glass" "None" :!class /Script/Engine.PrimitiveComponent))
      (Variables|Default|SetMID_LampGlass _returnvalue_2))
    (:then_2
      (bind _returnvalue_3 (Rendering|Material|CreateDynamicMaterialInstance (Variables|Default|GetBlackMonday_Strawberry_Arms) 0 "/Game/Gadgets/Assets/BlackMonday_Strawberry/Materials/MI_BlackMonday_Strawberry.MI_BlackMonday_Strawberry" "None" :!class /Script/Engine.PrimitiveComponent))
      (Variables|Default|SetMID_LampBase _returnvalue_3))))
